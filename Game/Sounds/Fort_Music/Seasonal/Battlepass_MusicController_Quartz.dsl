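; Battle Pass music controller Blueprint that manages the lifetime of a Quartz music track component.
; On BeginPlay, it queues the track to start when bUseLegacyImplementation is set, and otherwise destroys the component if valid.
; When the actor is destroyed, it queues a track stop on any valid component.
; The upsell music and QueueStop functions have empty bodies.
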(event EventBeginPlay
  (|CallParentFunction)
  (if (Variables|Default|GetbUseLegacyImplementation)
    (Control|QueueTrackStart (Variables|Default|GetQuartzMusicTrackComponent))
    (else
      (bind _quartzmusictrackcomponent (Variables|Default|GetQuartzMusicTrackComponent)
        (:then
          (Components|DestroyComponent _quartzmusictrackcomponent))
        (:else)))))

(event EventDestroyed
  (bind _quartzmusictrackcomponent (Variables|Default|GetQuartzMusicTrackComponent)
    (:then
      (Control|QueueTrackStop _quartzmusictrackcomponent))
    (:else)))

(fn PlayUpsellMusic
  (params (bPlay :type Boolean)))

(fn QueueStop)

(fn UpdateUpsellMusic
  (params (ParamName :type Name) (bIsSubscribed :type Boolean)))

(fn ConstructionScript
  (|CallParentFunction))
