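# Quad MetaSound source for seasonal Battle Pass menu music, crossfading between pre-purchase, upsell, subscription and crew loops from Season 34.
# All four loops start together on play and feed a stereo mixer, so track changes only move gains and keep the loops in step.
# Each track trigger raises its own gain and resets the other three, with short interpolation.
# The mix feeds the front pair, with an attenuated copy in the surrounds.
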MetaSoundSource /Game/Sounds/Fort_Music/Seasonal/MSS_Battlepass_Menu_Music
inputs [
 BPTrack {type=Float}
 Crew_Loop {type=WaveAsset:Array}
 CrewTrack {type=Float}
 OnBPTrack {type=Trigger}
 OnCrewTrack {type=Trigger}
 OnSubTrack {type=Trigger}
 OnUpSaleTrack {type=Trigger}
 PrePurchase_Loop {type=WaveAsset:Array}
 Subscription_Loop {type=WaveAsset:Array}
 SubTrack {type=Float}
 UE.Source.OnPlay {type=Trigger}
 UpSaleTrack {type=Float}
 Upsell_Loop {type=WaveAsset:Array}
]
interfaces [UE.OutputFormat.Quad:1.0, UE.Source:1.1]
outputs [
 "UE.OutputFormat.Quad.Audio:0" {type=Audio}
 "UE.OutputFormat.Quad.Audio:1" {type=Audio}
 "UE.OutputFormat.Quad.Audio:4" {type=Audio}
 "UE.OutputFormat.Quad.Audio:5" {type=Audio}
]
pages [Default {
  input_defaults {
   BPTrack=1
   CrewTrack=0
   OnBPTrack=null
   OnCrewTrack=null
   OnSubTrack=null
   OnUpSaleTrack=null
   SubTrack=0
   UE.Source.OnPlay=null
   UpSaleTrack=0
   Crew_Loop [/Game/Sounds/Fort_Music/Seasonal/S34/S34_FTUE_Loop_04_Crew_109BPM]
   PrePurchase_Loop [/Game/Sounds/Fort_Music/Seasonal/S34/S34_FTUE_Loop_03_PrePurchase_109BPM]
   Subscription_Loop [/Game/Sounds/Fort_Music/Seasonal/S34/S34_FTUE_Loop_01_Subscription_109BPM]
   Upsell_Loop [/Game/Sounds/Fort_Music/Seasonal/S34/S34_FTUE_Loop_02_Upsell_109BPM]
  }
  nodes [
   TriggerAnyTriggerAny3_0 = "TriggerAny.Trigger Any (3)"(In 1=^OnBPTrack, In 0=^OnCrewTrack, In 2=^OnUpSaleTrack)
   n1 = VariableAccessor.Trigger(Variable=VariableMutator.Trigger(Variable=InitVariable.Trigger().Variable, Value=@TriggerAnyTriggerAny3_0.Out).Variable)
   TriggerAnyTriggerAny3_2 = "TriggerAny.Trigger Any (3)"(In 0=^OnBPTrack, In 2=^OnSubTrack, In 1=^OnUpSaleTrack)
   n2 = VariableAccessor.Trigger(Variable=VariableMutator.Trigger(Variable=InitVariable.Trigger().Variable, Value=@TriggerAnyTriggerAny3_2.Out).Variable)
   TriggerAnyTriggerAny3_3 = "TriggerAny.Trigger Any (3)"(In 2=^OnCrewTrack, In 1=^OnSubTrack, In 0=^OnUpSaleTrack)
   n3 = VariableAccessor.Trigger(Variable=VariableMutator.Trigger(Variable=InitVariable.Trigger().Variable, Value=@TriggerAnyTriggerAny3_3.Out).Variable)
   TriggerAnyTriggerAny3_1 = "TriggerAny.Trigger Any (3)"(In 2=^OnBPTrack, In 1=^OnCrewTrack, In 0=^OnSubTrack)
   n4 = VariableAccessor.Trigger(Variable=VariableMutator.Trigger(Variable=InitVariable.Trigger().Variable, Value=@TriggerAnyTriggerAny3_1.Out).Variable)
   n5 = /FortMetasound/Patches/MSP_Audio_WavePlayer_Basic_Stereo(Play Trigger=^UE.Source.OnPlay, Wave Asset=^Upsell_Loop) {also {defaults {Is Looping?="<MetaSoundBuilderResult.SUCCEEDED: 0>"}}}
   n6 = &MSP_Audio_WavePlayer_Basic_Stereo(Wave Asset=^PrePurchase_Loop, Play Trigger=^UE.Source.OnPlay) {also {defaults {Is Looping?="<MetaSoundBuilderResult.SUCCEEDED: 0>"}}}
   n7 = &MSP_Audio_WavePlayer_Basic_Stereo(Wave Asset=^Subscription_Loop, Play Trigger=^UE.Source.OnPlay) {also {defaults {Is Looping?="<MetaSoundBuilderResult.SUCCEEDED: 0>"}}}
   n8 = &MSP_Audio_WavePlayer_Basic_Stereo(Wave Asset=^Crew_Loop, Play Trigger=^UE.Source.OnPlay) {also {defaults {Is Looping?="<MetaSoundBuilderResult.SUCCEEDED: 0>"}}}
   AudioMixerAudioMixerStereo4_0 = "AudioMixer.Audio Mixer (Stereo, 4)"(
    Gain 2=UE.InterpTo:Audio(Target=Value.Value:Float(Set=^OnSubTrack, Reset=@n1.Value) {also {defaults {Init Value=0, Target Value=1}}}.Output Value) {also {defaults {Interp Time=0.15}}}.Value
    Gain 3=UE.InterpTo:Audio(Target=Value.Value:Float(Set=^OnCrewTrack, Reset=@n2.Value) {also {defaults {Init Value=0, Target Value=1}}}.Output Value) {also {defaults {Interp Time=0.15}}}.Value
    Gain 0=UE.InterpTo:Audio(Target=Value.Value:Float(Set=^OnBPTrack, Reset=@n3.Value) {also {defaults {Target Value=1}}}.Output Value) {also {defaults {Interp Time=0.15}}}.Value
    Gain 1=UE.InterpTo:Audio(Target=Value.Value:Float(Set=^OnUpSaleTrack, Reset=@n4.Value) {also {defaults {Init Value=0, Target Value=1}}}.Output Value) {also {defaults {Interp Time=0.15}}}.Value
    In 1 L=@n5.Out L
    In 1 R=@n5.Out R
    In 0 L=@n6.Out L
    In 0 R=@n6.Out R
    In 2 L=@n7.Out L
    In 2 R=@n7.Out R
    In 3 L=@n8.Out L
    In 3 R=@n8.Out R
   )
   Multiply_0 = UE.Multiply:Audio by Float(PrimaryOperand=@AudioMixerAudioMixerStereo4_0.Out R) {also {defaults {AdditionalOperands=0.25}}}
   Multiply_1 = UE.Multiply:Audio by Float(PrimaryOperand=@AudioMixerAudioMixerStereo4_0.Out L) {also {defaults {AdditionalOperands=0.25}}}
  ]
  links [
   {dst="@UE.OutputFormat.Quad.Audio:0", src=AudioMixerAudioMixerStereo4_0.Out L}
   {dst="@UE.OutputFormat.Quad.Audio:1", src=AudioMixerAudioMixerStereo4_0.Out R}
   {dst="@UE.OutputFormat.Quad.Audio:5", src=Multiply_0.Out}
   {dst="@UE.OutputFormat.Quad.Audio:4", src=Multiply_1.Out}
  ]
 }]
properties {
 assetImportData=&self:AssetImportData
 bLooping=true
 bOverrideConcurrency=true
 duration=10000
 numChannels=4
 outputFormat=Quad
 sampleRate=48000
 soundClassObject=&SFX
 soundSubmixObject=&Fortnite_SFX
 virtualizationMode=PlayWhenSilent
 volume=0.65
 concurrencyOverrides {maxCount=1, resolutionRule=StopOldest}
}
