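# CineToBP_Transition_Cue is a Sound Cue among seasonal Fortnite music assets, named for a cinematic-to-BP transition.
# Its graph is empty, with no root node and no sound nodes, so it plays no audio as saved.
# It is assigned to the MediaCinematics sound class and sets a very large max distance.
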SoundCue /Game/Sounds/Fort_Music/Seasonal/CineToBP_Transition_Cue
root=null
nodes []
properties {maxDistance=2097152, soundClassObject=/Game/Sounds/FortSoundClassesAndModes/MediaCinematics, volumeMultiplier=1}
